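# Master material for the NeonJam backpack cosmetic, pulsing its shape and glow to music over the shared character skin and FX functions.
# Music spectrum and amplitude from a parameter collection, or a static-switched fake value, become a quantized Audio_Amplitude.
# SkinFX_Mask yields crown, cat-ear and yellow-area masks for the effects.
# World position offset scales from an offset center; emissive combines panning caustics, clouds and fresnel, tinted and HDR-boosted.
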Material /BRCosmetics/Accessories/FORT_Backpacks/Backpack_F_MED_NeonJam/Material/F_MED_NeonJam_Backpack_Master
graph {
 properties {
  bAutomaticallySetUsageInEditor=false
  bUseMaterialAttributes=true
  bUsedWithMeshDeformer=true
  bUsedWithMorphTargets=true
  bUsedWithSkeletalMesh=true
  pixelDepthOffsetMode=PDOM_Legacy
  refractionMethod=RM_IndexOfRefraction
  shadingModel=MSM_FromMaterialExpression
  subsurfaceProfile=&SS_HeroSkin_02
  usedShadingModels=Default Lit | Subsurface | Clear Coat | Subsurface Profile | Cloth | Eye
 }
 nodes [
  n1 = MaterialFunctionCall() {materialFunction=&BaseFN_CharacterFunction_PreFX}
  n2 = GetMaterialAttributes(None=MaterialFunctionCall(
    PreskinnedPosition=@n1.Pre skinned Local Position
    Gmap=@n1.Pre skinned Local Position
    MaterialAttributes=@n1.Pre skinned Local Position
    AO=@n1.Pre skinned Local Position
   ) {materialFunction=&CharacterSkinFX}.Result) {attributeGetTypes [#1, #2, #3]}
  n3 = NamedRerouteUsage()
  n4 = NamedRerouteUsage()
  n5 = MaterialFunctionCall(
   AO=@n1.AO
   Distance Blend=@n1.AO
   Fresnel=@n1.AO
   Gmap=@n1.AO
   Pre Skinned Local Position=@n1.AO
   Material=SetMaterialAttributes!(
    MaterialAttributes=@n2.MaterialAttributes
    World Position Offset=Add(A=@n2.World Position Offset, B=@n3)
    Emissive Color=Add(A=@n2.Emissive Color, B=@n4)
   ) {attributeSetTypes [#1, #2]}
  ) {materialFunction=&BaseFN_CharacterFunction_PostFX}
  MaterialFunctionCall(In=@n5.Result) {materialFunction=&CINE_MF_CharPostFX}
  n6 = ScalarParameter() {defaultValue=10, group=NeonJam, parameterName=QuantizeAudioSteps}
  n7 = MaterialFunctionCall(Float4=CollectionParameter() {collection=/Game/Athena/Events/AudioAnalysis/MPC_Audio_Reactive_Parameters, parameterName=Spectrum_Music}) {materialFunction=&BreakOutFloat4Components}
  n8 = Add(A=Multiply(A=@n7.R, B=ScalarParameter() {defaultValue=2, group=NeonJam, parameterName=Base_BoostAmount}), B=@n7.G)
  n9 = Multiply(A=Add(A=Add(A=@n8, B=@n7.B), B=@n7.A)) {constB=0.75}
  n10 = CollectionParameter() {collection=&MPC_Audio_Reactive_Parameters, parameterName=Amplitude_Music}
  n11 = Multiply(
   A=@n6
   B=StaticSwitchParameter(True=ScalarParameter() {defaultValue=1, group=NeonJam, parameterName=Fake_Audio_Amplitude}, False=Multiply(A=@n9, B=@n10)) {group=NeonJam, parameterName=UseFakeAudioData}
  ) {constB=5}
  n12 = Multiply(A=Divide(B=@n6), B=Subtract(A=Floor(None=@n11)))
  n13 = ScalarParameter() {defaultValue=1, group=NeonJam, parameterName=AudioReactiveMultiplier}
  NamedRerouteDeclaration(None=Max(A=Multiply(A=@n12, B=@n13)) {constB=0}) {name=Audio_Amplitude, nodeColor=<rgba 0.5938 0 1 1>}
  n14 = Ceil(None=NamedRerouteUsage())
  n15 = ScalarParameter() {defaultValue=0.45, group=NeonJam, parameterName=Include_Crown_In_WPO}
  n16 = Saturate(None=Add(A=Multiply(A=@n14, B=@n15), B=NamedRerouteUsage()))
  n17 = NamedRerouteUsage()
  n18 = Subtract(
   A=WorldPosition().XYZ
   B=TransformPosition(None=VectorParameter() {group=NeonJam, parameterName=CenterOffset, defaultValue=<rgba 0 -10 -5 0>}.RGB) {transformType=TRANSFORMPOSSOURCE_World}
  )
  n19 = ScalarParameter() {defaultValue=0.25, group=NeonJam, parameterName=Audio_WPO_Intensity}
  NamedRerouteDeclaration(None=Multiply(A=Multiply(A=@n16, B=@n17), B=Multiply(A=@n18, B=@n19))) {name=WPO_Out, nodeColor=<rgba 0.4297 0 1 1>}
  n20 = VertexInterpolator(VS=PreSkinnedPosition())
  n21 = ComponentMask(None=Divide(A=Abs(None=@n20.PS), B=ScalarParameter() {defaultValue=15, group=NeonJam, parameterName=Breakup_Tile_WS}) {constB=15}) {b=true, r=true}
  n22 = Panner(Coordinate=@n21) {speedX=-0.2, speedY=-0.2}
  n23 = TextureSample(UVs=Panner(Coordinate=@n21) {speedX=-0.33, speedY=-0.33}) {samplerSource=SSM_Wrap_WorldGroupSettings, samplerType=SAMPLERTYPE_LinearGrayscale, texture=/Game/Athena/Prototype/Terrain/T_CloudCoverage}
  n24 = Multiply(A=@n23.G) {constB=0.5}
  n25 = Saturate(None=NamedRerouteUsage())
  n26 = Abs(None=Subtract(
    A=TextureSample(UVs=@n22) {samplerSource=SSM_Wrap_WorldGroupSettings, texture=/Game/Global/Textures/Default/Caustics/T_Caustics_Softer}.G
    B=Multiply(A=@n24, B=@n25)
   ))
  n27 = Multiply(A=NamedRerouteUsage(), B=@n26)
  n28 = ScalarParameter() {defaultValue=0.2, group=NeonJam, parameterName=Cat_Ears_Emissive_Boost}
  n29 = Constant3Vector() {constant=<rgba 1 0 0 0>}
  n30 = PixelNormalWS()
  n31 = Power(Base=Abs(None=DotProduct(A=@n29, B=@n30))) {constExponent=0.3}
  n32 = Multiply(A=OneMinus(None=Saturate(None=@n31)), B=@n26)
  n33 = Fresnel() {baseReflectFraction=0.1, exponent=4}
  n34 = Add(A=Add(A=Multiply(A=@n27, B=@n28), B=Multiply(A=@n32) {constB=0.5}), B=@n33)
  n35 = NamedRerouteUsage()
  n36 = Saturate(None=@n35)
  n37 = Multiply(A=@n33, B=@n35)
  n38 = Multiply(A=Multiply(A=LinearInterpolate(A=@n37, B=@n34, Alpha=@n36), B=Ceil(None=@n36)), B=NamedRerouteUsage())
  n39 = VectorParameter() {parameterName=Emissive_Accent_Color, defaultValue=<rgba 0.9375 0.3865 0 0>}
  n40 = ScalarParameter() {defaultValue=2, group=NeonJam, parameterName=Emissive_Accent_Intensity}
  n41 = NamedRerouteUsage()
  NamedRerouteDeclaration(None=MaterialFunctionCall(LDR Input=Multiply(A=@n38, B=@n39.RGB), HDR Intensity Multiplier=Multiply(A=@n40, B=@n41)) {materialFunction=&DeriveHDRfromLDR}.RemappedRange) {name=Emissive_Out, nodeColor=<rgba 0 0.8125 1 1>}
  n42 = TextureSampleParameter2D() {
   parameterName=SkinFX_Mask
   samplerSource=SSM_Wrap_WorldGroupSettings
   texture=/BRCosmetics/Accessories/FORT_Backpacks/Backpack_F_MED_NeonJam/Texture/T_F_MED_NeonJam_Backpack_FX
  }
  n43 = Subtract(A=@n42.R) {constB=0.15}
  n44 = Multiply(A=@n43) {constB=-1}
  NamedRerouteDeclaration(None=Multiply(A=@n42.G, B=Saturate(None=@n44)) {desc=Crown Mask}) {name=Crown_Mask, nodeColor=<rgba 1 0 0.2812 1>}
  NamedRerouteDeclaration(None=Ceil(None=@n43) {desc=All but Crown}) {name=All_But_Crown_Mask, nodeColor=<rgba 0 0.5078 1 1>}
  NamedRerouteDeclaration(None=Ceil(None=Subtract(A=@n42.R) {constB=0.5}) {desc=Yellow Areas for Emissive}) {name=Yellow_Areas_Mask, nodeColor=<rgba 0.7578 0 1 1>}
  NamedRerouteDeclaration(None=Ceil(None=Subtract(A=@n42.R) {constB=0.9})) {name=Cat_Ears_Mask, nodeColor=<rgba 1 0.6797 0 1>}
  n45 = VectorParameter() {group=NeonJam, parameterName=Fake_Audio_Bands, defaultValue=<rgba 0.5 0.9 0.1 0.75>}
  StaticSwitchParameter(
   True=OneMinus(None=AppendVector(A=@n45.RGB, B=@n45.RGB))
   False=MaterialFunctionCall!(useSmoothStep=StaticBool(), uvx=Constant() {r=0.5}) {materialFunction=/Game/Accessories/FORT_Backpacks/Backpack_F_MED_Cactus_Rocker/Materials/SampleAudioMPCHistory}.Interpolated Spectrum
  ) {group=NeonJam, parameterName=UseFakeAudioData}
 ]
}
