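# Physics asset for the NeonJam backpack cosmetic, defining secondary motion for its head, crown and headphones pieces.
# Four small capsule bodies cover root and the three dyn_ bones, with the head constrained to root by tight angular limits.
# Crown and headphones both attach to the head: the crown is locked on all angular axes, while the headphones swing widely on one axis and barely on the others.
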PhysicsAsset /BRCosmetics/Accessories/FORT_Backpacks/Backpack_F_MED_NeonJam/Meshes/F_MED_NeonJam_Pack_Physics
bodies {
 dyn_crown [{shapeType=Capsule, length=2, radius=2, shapeName=dyn_crown_capsule}]
 dyn_head [{shapeType=Capsule, length=2, radius=2, shapeName=dyn_head_capsule}]
 dyn_headphones [{shapeType=Capsule, length=2, radius=2, shapeName=dyn_headphones_capsule}]
 root [{shapeType=Capsule, length=2, radius=2, shapeName=root_capsule}]
}
constraints [
 {bone1Name=dyn_head, bone2Name=root, swing1LimitDegrees=3, swing2LimitDegrees=5, twistLimitDegrees=5}
 {
  bone1Name=dyn_crown
  bone2Name=dyn_head
  swing1LimitDegrees=3
  swing1Motion=Locked
  swing2LimitDegrees=10
  swing2Motion=Locked
  twistLimitDegrees=5
  twistMotion=Locked
 }
 {bone1Name=dyn_headphones, bone2Name=dyn_head, swing1LimitDegrees=30, swing2LimitDegrees=3, twistLimitDegrees=2}
]
