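# Animation Blueprint for the NeonJam backpack cosmetic, adding secondary motion and an idle float to its crown and headphones.
# The event graph builds a looping Crown_Float offset from real-time sine waves.
# The AnimGraph runs Rigid Body physics, adds that float and a fixed offset to the crown bone, then applies spring-based Anim Dynamics.
# Crown dynamics blend at partial strength, while headphones dynamics run at full strength.
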AnimBlueprint /BRCosmetics/Accessories/FORT_Backpacks/Backpack_F_MED_NeonJam/Meshes/F_MED_NeonJam_Pack_AnimBP
target_skeleton=&/F_MED_NeonJam_Pack_Skeleton
graphs [
 AnimGraph {nodes [
   n1 = RigidBody(ComponentPose=Local To Component(LocalPose=Input Pose())) {
    overridePhysicsAsset=&/F_MED_NeonJam_Pack_Physics
    simulationSpace=BaseBoneSpace
    alphaBoolBlend {blendOption=Linear}
    baseBoneRef {boneName=root}
    componentLinearAccScale=<1 1 1>
   }
   n2 = "Transform (Modify) Bone - Bone: dyn_crown_base"(ComponentPose=@n1, Alpha=1, Translation=Get Crown_Float() -> (Crown_Float:Vector), Rotation="0,0,0", Scale="1,1,1") {translationMode=BMM_Additive, translationSpace=BCS_BoneSpace, alphaBoolBlend {blendOption=Linear}, boneToModify {boneName=dyn_crown_base}}
   n3 = "Transform (Modify) Bone - Bone: dyn_crown_base"(ComponentPose=@n2, Alpha=1, Translation="-1,0,0", Rotation="0,0,0", Scale="1,1,1") {
    translationMode=BMM_Additive
    translationSpace=BCS_BoneSpace
    alphaBoolBlend {blendOption=Linear}
    boneToModify {boneName=dyn_crown_base}
    translation {x=-1}
   }
   n4 = Anim Dynamics - Bone: dyn_crown_base(ComponentPose=@n3, Alpha=0.75, ExternalForce="0,0,0") {
    alpha=0
    angularBiasOverride=1
    angularDampingOverride=0.75
    angularSpringConstant=30
    bAngularSpring=true
    bGravityOverrideInSimSpace=true
    bLinearSpring=true
    bOverrideAngularBias=true
    bOverrideAngularDamping=true
    bOverrideLinearDamping=true
    bUseGravityOverride=true
    bUsePlanarLimit=false
    bUseSphericalLimits=true
    gravityScale=0
    linearDampingOverride=0.75
    linearSpringConstant=30
    simulationSpace=BoneRelative
    alphaBoolBlend {blendOption=Linear}
    boundBone {boneName=dyn_crown_base}
    componentAppliedLinearAccClamp=<90 90 90>
    componentLinearAccScale=<0.26 0.26 0.26>
    componentLinearVelScale=<0.26 0.26 0.26>
    gravityOverride {z=-35}
    physicsBodyDefinitions [{
      sphereCollisionRadius=0
      boundBone {boneName=dyn_crown_base}
      boxExtents=<10 10 10>
      constraintSetup {
       linearXLimitType=Free
       linearYLimitType=Free
       linearZLimitType=Free
       angularLimitsMax=<3 0 0>
       angularLimitsMin=<-3 0 0>
       angularTarget=<-10 0 0>
       linearAxesMax=<3 2 2.5>
       linearAxesMin=<-10 -2 -2.5>
      }
      localJointOffset=<0 0 0>
     }]
    relativeSpaceBone {boneName=root}
    sphericalLimits [{limitRadius=2, limitType=Inner, drivingBone {boneName=dyn_crown_base}, sphereLocalOffset=<0 0 0>}]
   }
   n5 = Anim Dynamics - Bone: dyn_headphones(ComponentPose=@n4, Alpha=1, ExternalForce="0,0,0") {
    alpha=0
    angularBiasOverride=1
    angularDampingOverride=0.9
    angularSpringConstant=0.5
    bAngularSpring=true
    bGravityOverrideInSimSpace=true
    bLinearSpring=true
    bOverrideAngularBias=true
    bOverrideAngularDamping=true
    bOverrideLinearDamping=true
    bUseGravityOverride=true
    gravityScale=0
    linearDampingOverride=0.9
    linearSpringConstant=0.5
    simulationSpace=BoneRelative
    alphaBoolBlend {blendOption=Linear}
    boundBone {boneName=dyn_headphones}
    componentAppliedLinearAccClamp=<33 33 33>
    componentLinearAccScale=<0.13 0.13 0.13>
    componentLinearVelScale=<0.13 0.13 0.13>
    gravityOverride {z=-300}
    physicsBodyDefinitions [{
      sphereCollisionRadius=0
      boundBone {boneName=dyn_headphones}
      boxExtents=<6 6 6>
      constraintSetup {angularLimitsMax=<0 0 20>, angularLimitsMin=<0 0 -20>, angularTarget=<0 0 0>}
      localJointOffset=<2 -0.5 0>
     }]
    relativeSpaceBone {boneName=root}
   }
   Output Pose(Result=Component To Local(ComponentPose=@n5)) {layerGroup=None}
  ]}
 EventGraph {nodes [
   Event BlueprintUpdateAnimation()
   > Sequence!(execute) -> (then_0, then_1, then_2:Exec, then_3:Exec, then_4:Exec, then_5:Exec, then_6:Exec)
   n1 = GetRealTimeSeconds()
   n2 = "Sin (Radians)"(A=float * float(A=float * float(A=1, B=1), B=@n1))
   n3 = "Sin (Radians)"(A=float * float(A=float * float(A=1, B=1), B=@n1))
   n4 = "Sin (Radians)"(A=float * float(A=float * float(A=1, B=1.5), B=@n1))
   > Set Crown_Float(Crown_Float:Vector=MakeVector(X=float * float(A=@n4, B=0.9), Y=float * float(A=@n2, B=0.3), Z=float * float(A=@n3, B=0.2))) -> (Output_Get:Vector)
   TryGetPawnOwner()
  ]}
]
variables [Crown_Float {type=Vector}]
